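# Conversation database for the Peely Bone NPC, providing its greeting and top-level conversation routing.
# The entry point feeds one speech node whose default line can be replaced by contextual messages for a first-ever meeting, match conversation count, or two banana outfits.
# First-meeting and outfit lines use named priority curve rows.
# The speech node then branches through link nodes to the shared quest and basic-services entries and Peely Bone's dialogue entry.
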ConversationDatabase /NPCLibrary/NPCs/PeelyBone/NPCLibrary_Conversation_PeelyBone
entry_points [{tag=Conversation.Entry.PeelyBone, destinations [#1]}]
nodes [
 ConversationEntryPointNode_7 {
  id=#1
  class=ConversationEntryPointNode
  outputs [#2]
  properties {entryTag {tagName=Conversation.Entry.PeelyBone}}
 }
 Speech_C_4 {
  id=#2
  class=Speech_C
  outputs [#3, #4, #5]
  properties {
   nodeName=""
   generalConfig {
    defaultMessage=(Peely Bone nods hello. The motion makes his brain slops against his head-ribs.)
    contextualMessages [
     {
      message=(Peely Bone waves his bones at you!)
      contextRequirements [{requirement {class=FortControllerRequirement_IsFirstEverConversationWithNPC}}]
      priority {curve {rowName=Tandem.Speech.Priority.FirstEverInteraction}}
     }
     {
      message=(Peely Bone nods hello. The motion makes his brain slop against his head-ribs.)
      contextRequirements [{requirement {class=FortControllerRequirement_CheckNPCMatchConversationCount, properties {nPCUniqueIdTag {tagName=AISpawnerData.Type.Tandem.PeelyBone}}}}]
      priority {curve {rowName=None}}
     }
     {
      message=(Urgh... looks like Peely Bone's just finished lunch...)
      contextRequirements [{requirement {class=FortControllerRequirement_CheckNPCMatchConversationCount, properties {nPCUniqueIdTag {tagName=AISpawnerData.Type.Tandem.PeelyBone}}}}]
      priority {curve {rowName=None}}
     }
     {
      message="(Peely Bone grins. Half Peely, half bone, all smiles!"
      contextRequirements [{requirement {class=FortControllerRequirement_CheckNPCMatchConversationCount, properties {nPCUniqueIdTag {tagName=AISpawnerData.Type.Tandem.PeelyBone}}}}]
      priority {curve {rowName=None}}
     }
     {
      message=(Peely Bone nods serenely. He has found peace in his incompleteness.)
      contextRequirements [{requirement {
         class=FortControllerRequirement_HasCID
         properties {allowedCIds [&Character_BananaPhilosopher]}
        }}]
      priority {curve {rowName=Tandem.Speech.Priority.OutfitReaction}}
     }
     {
      message=(Peely Bone looks confused. Where are your organs?)
      contextRequirements [{requirement {
         class=FortControllerRequirement_HasCID
         properties {allowedCIds [&CID_349_Athena_Commando_M_Banana]}
        }}]
      priority {curve {rowName=Tandem.Speech.Priority.OutfitReaction}}
     }
    ]
   }
  }
 }
 ConversationLinkNode_3 {id=#5, class=ConversationLinkNode, properties {remoteEntryTag {tagName=Conversation.Entry.PeelyBone.Dialogue}}}
 ConversationLinkNode_0 {id=#4, class=ConversationLinkNode, properties {remoteEntryTag {tagName=Conversation.Entry.BasicServices}}}
 ConversationLinkNode_1 {id=#3, class=ConversationLinkNode, properties {remoteEntryTag {tagName=Conversation.Entry.NPCLibrary.Quests}}}
]
